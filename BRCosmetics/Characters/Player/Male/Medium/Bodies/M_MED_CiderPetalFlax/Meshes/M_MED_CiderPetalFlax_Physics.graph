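# Physics asset for the M_MED_CiderPetalFlax male medium player body, set up around the jacket's bottom and front flaps.
# Capsules on the pelvis, lower spine, thighs and feet form the body volumes, linked by free-swinging, free-twisting constraints.
# Short jacket bone chains carry capsules with narrow swing limits and one-degree twist.
# Only the two A chains constrain to spine_01, and each also links freely to its thigh.
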PhysicsAsset /BRCosmetics/Characters/Player/Male/Medium/Bodies/M_MED_CiderPetalFlax/Meshes/M_MED_CiderPetalFlax_Physics
bodies {
 base_bottomjacket_A_l [{shapeType=Capsule, length=0.3, radius=0.5, shapeName=MSTR_floor_A_l_capsule}]
 base_bottomjacket_A_r [{shapeType=Capsule, length=0.3, radius=0.5, shapeName=MSTR_floor_A_r_capsule}]
 base_bottomjacket_B_l [{shapeType=Capsule, length=0.3, radius=0.5, shapeName=MSTR_floor_B_l_capsule}]
 base_bottomjacket_B_r [{shapeType=Capsule, length=0.3, radius=0.5, shapeName=MSTR_floor_B_r_capsule}]
 base_bottomjacket_Mid [{shapeType=Capsule, length=0.3, radius=0.5, shapeName=MSTR_floor_Mid_capsule}]
 base_frontjacket_A_l [{shapeType=Capsule, length=0.505, radius=0.505, shapeName=base_frontjacket_A_l_capsule, rotation=<rot 0 -90 0>}]
 bottomjacket_A_l_1 [{shapeType=Capsule, length=8.9859, radius=5.8332, shapeName=floor_A_l_1_capsule, center=<7.354 0 5.9427>, rotation=<rot 90 0 0>}]
 bottomjacket_A_r_1 [{shapeType=Capsule, length=9.4683, radius=5.8332, shapeName=floor_A_r_1_capsule, center=<6.9945 0 6.6342>, rotation=<rot 90 0 0>}]
 bottomjacket_B_l_1 [{shapeType=Capsule, length=0.3534, radius=5.1949, shapeName=floor_B_l_1_capsule, center=<3.4344 0 4.4193>, rotation=<rot 90 0 0>}]
 bottomjacket_B_l_2 [{shapeType=Capsule, length=0.3534, radius=5.1949, shapeName=floor_B_l_2_capsule, center=<0.062 0 4.4193>, rotation=<rot 90 0 0>}]
 bottomjacket_B_l_3 [{shapeType=Capsule, length=0.5366, radius=5.1949, shapeName=floor_B_l_3_capsule, center=<2.2593 0 4.4193>, rotation=<rot 90 0 0>}]
 bottomjacket_B_r_1 [{shapeType=Capsule, length=0.3534, radius=5.1949, shapeName=floor_B_r_1_capsule, center=<3.4344 0 4.4193>, rotation=<rot 90 0 0>}]
 bottomjacket_B_r_2 [{shapeType=Capsule, length=0.3534, radius=5.1949, shapeName=floor_B_r_2_capsule, center=<0.062 0 4.4193>, rotation=<rot 90 0 0>}]
 bottomjacket_B_r_3 [{shapeType=Capsule, length=0.5366, radius=5.1949, shapeName=floor_B_r_3_capsule, center=<2.2593 0 4.4193>, rotation=<rot 90 0 0>}]
 bottomjacket_Mid_l_1 [{shapeType=Capsule, length=3, radius=2, shapeName=floor_Mid_l_1_capsule, center=<1 0 1.8021>, rotation=<rot 90 0 0>}]
 bottomjacket_Mid_l_2 [{shapeType=Capsule, length=3, radius=2, shapeName=floor_Mid_l_2_capsule, center=<1 0 1.8021>, rotation=<rot 90 0 0>}]
 bottomjacket_Mid_l_3 [{shapeType=Capsule, length=3, radius=2, shapeName=floor_Mid_l_3_capsule, center=<5 0 1.8021>, rotation=<rot 90 0 0>}]
 foot_l [{shapeType=Capsule, length=9.9983, radius=7.4739, shapeName=foot_l_capsule, center=<-4.7403 2.5124 0>, rotation=<rot 0 0 90>}]
 foot_r [{shapeType=Capsule, length=10.5184, radius=6.8082, shapeName=foot_r_capsule, center=<5.3053 -3.7148 0>, rotation=<rot 0 0 90>}]
 frontjacket_A_l_01 [{shapeType=Capsule, length=0.3534, radius=5.1949, shapeName=floor_B_l_1_capsule, center=<3.4344 0 4.4193>, rotation=<rot 90 0 0>}]
 frontjacket_A_l_02 [{shapeType=Capsule, length=0.3534, radius=5.1949, shapeName=floor_B_l_2_capsule, center=<0.062 0 4.4193>, rotation=<rot 90 0 0>}]
 pelvis [{shapeType=Capsule, length=9.6457, radius=11.85, shapeName=pelvis_capsule, center=<-5.8218 0.3762 0>, rotation=<rot 0 -90 0>}]
 spine_01 [{shapeType=Capsule, length=8.7255, radius=11.3247, shapeName=spine_01_capsule, center=<-2.6741 0.3228 -0.1724>, rotation=<rot 0 -90 0>}]
 thigh_l [{shapeType=Capsule, length=7.4759, radius=8.7205, shapeName=thigh_l_capsule, center=<-8.7633 -0.7769 0.8451>, rotation=<rot 90 -5.6389 0>}]
 thigh_r [{shapeType=Capsule, length=7.3472, radius=8.7766, shapeName=thigh_r_capsule, center=<9.1198 1.7718 -1.037>, rotation=<rot 90 -5.2443 0>}]
}
constraints [
 {
  bone1Name=spine_01
  bone2Name=pelvis
  swing1LimitDegrees=45
  swing1Motion=Free
  swing2LimitDegrees=45
  swing2Motion=Free
  twistLimitDegrees=45
  twistMotion=Free
 }
 {
  bone1Name=thigh_l
  bone2Name=pelvis
  swing1LimitDegrees=45
  swing1Motion=Free
  swing2LimitDegrees=45
  swing2Motion=Free
  twistLimitDegrees=45
  twistMotion=Free
 }
 {
  bone1Name=thigh_r
  bone2Name=pelvis
  swing1LimitDegrees=45
  swing1Motion=Free
  swing2LimitDegrees=45
  swing2Motion=Free
  twistLimitDegrees=45
  twistMotion=Free
 }
 {bone1Name=base_bottomjacket_A_l, bone2Name=spine_01, swing1LimitDegrees=15, swing2LimitDegrees=10, twistLimitDegrees=1}
 {bone1Name=bottomjacket_A_l_1, bone2Name=base_bottomjacket_A_l, swing1LimitDegrees=15, swing2LimitDegrees=10, twistLimitDegrees=1}
 {
  bone1Name=thigh_l
  bone2Name=bottomjacket_A_l_1
  swing1LimitDegrees=45
  swing1Motion=Free
  swing2LimitDegrees=45
  swing2Motion=Free
  twistLimitDegrees=45
  twistMotion=Free
 }
 {bone1Name=base_bottomjacket_A_r, bone2Name=spine_01, swing1LimitDegrees=15, swing2LimitDegrees=10, twistLimitDegrees=1}
 {bone1Name=bottomjacket_A_r_1, bone2Name=base_bottomjacket_A_r, swing1LimitDegrees=15, swing2LimitDegrees=10, twistLimitDegrees=1}
 {
  bone1Name=thigh_r
  bone2Name=bottomjacket_A_r_1
  swing1LimitDegrees=45
  swing1Motion=Free
  swing2LimitDegrees=45
  swing2Motion=Free
  twistLimitDegrees=45
  twistMotion=Free
 }
 {
  bone1Name=foot_l
  bone2Name=thigh_l
  swing1LimitDegrees=45
  swing1Motion=Free
  swing2LimitDegrees=45
  swing2Motion=Free
  twistLimitDegrees=45
  twistMotion=Free
 }
 {
  bone1Name=foot_r
  bone2Name=thigh_r
  swing1LimitDegrees=45
  swing1Motion=Free
  swing2LimitDegrees=45
  swing2Motion=Free
  twistLimitDegrees=45
  twistMotion=Free
 }
 {bone1Name=bottomjacket_Mid_l_1, bone2Name=base_bottomjacket_Mid, swing1LimitDegrees=5, swing2LimitDegrees=20, twistLimitDegrees=1}
 {bone1Name=bottomjacket_Mid_l_2, bone2Name=bottomjacket_Mid_l_1, swing1LimitDegrees=5, swing2LimitDegrees=20, twistLimitDegrees=1}
 {bone1Name=bottomjacket_Mid_l_3, bone2Name=bottomjacket_Mid_l_2, swing1LimitDegrees=5, swing2LimitDegrees=20, twistLimitDegrees=1}
 {bone1Name=bottomjacket_B_l_2, bone2Name=bottomjacket_B_l_1, swing1LimitDegrees=5, swing2LimitDegrees=15, twistLimitDegrees=1}
 {bone1Name=bottomjacket_B_l_3, bone2Name=bottomjacket_B_l_2, swing1LimitDegrees=5, swing2LimitDegrees=15, twistLimitDegrees=1}
 {bone1Name=base_bottomjacket_B_l, bone2Name=bottomjacket_B_l_1, swing1LimitDegrees=5, swing2LimitDegrees=15, twistLimitDegrees=1}
 {bone1Name=bottomjacket_B_r_2, bone2Name=bottomjacket_B_r_1, swing1LimitDegrees=5, swing2LimitDegrees=15, twistLimitDegrees=1}
 {bone1Name=bottomjacket_B_r_3, bone2Name=bottomjacket_B_r_2, swing1LimitDegrees=5, swing2LimitDegrees=15, twistLimitDegrees=1}
 {bone1Name=bottomjacket_B_r_1, bone2Name=base_bottomjacket_B_r, swing1LimitDegrees=5, swing2LimitDegrees=15, twistLimitDegrees=1}
 {bone1Name=frontjacket_A_l_01, bone2Name=base_frontjacket_A_l, swing1LimitDegrees=5, swing2LimitDegrees=15, twistLimitDegrees=1}
 {bone1Name=frontjacket_A_l_02, bone2Name=frontjacket_A_l_01, swing1LimitDegrees=5, swing2LimitDegrees=15, twistLimitDegrees=1}
]
