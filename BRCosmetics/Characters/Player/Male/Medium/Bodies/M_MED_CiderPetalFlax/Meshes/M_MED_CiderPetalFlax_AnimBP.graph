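# Animation Blueprint for the M_MED_CiderPetalFlax male body cosmetic, adding secondary motion to its jacket, watch and buttons.
# The parent mesh pose passes through a Control Rig into a cache, feeding two bool-blended AnimDynamics front-jacket setups.
# Weighted constraints then place watch and button bones before a pelvis-space RigidBody pass.
# The update event sets crouch state and a gravity override computed from F_MED_AlderOptic's face-accessory parameters.
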AnimBlueprint /BRCosmetics/Characters/Player/Male/Medium/Bodies/M_MED_CiderPetalFlax/Meshes/M_MED_CiderPetalFlax_AnimBP
target_skeleton=&/M_MED_CiderPetalFlax_Skeleton
graphs [
 AnimGraph {nodes [
   n1 = Anim Dynamics - Bone: frontjacket_A_r_01(Alpha=0.6, ExternalForce="0,0,0") {
    angularDampingOverride=1
    angularSpringConstant=9000
    bAngularSpring=true
    bOverrideAngularDamping=true
    bUseGravityOverride=true
    gravityScale=0.35
    simulationSpace=RootRelative
    boundBone {boneName=frontjacket_A_r_01}
    componentLinearAccScale=<1 1 1>
    componentLinearVelScale=<1 1 1>
    gravityOverride {z=-980}
    physicsBodyDefinitions [{
      boundBone {boneName=frontjacket_A_r_01}
      boxExtents=<10 30 10>
      constraintSetup {angularLimitsMax=<0 0 5>, angularLimitsMin=<0 -15 -5>, angularTarget=<1 0 0>}
      localJointOffset=<2 0 0>
     }]
   }
   Anim Dynamics - Bone: frontjacket_A_r_02(ComponentPose=@n1, Alpha=0.4, ExternalForce="0,0,0") {
    angularDampingOverride=1
    angularSpringConstant=9000
    bAngularSpring=true
    bOverrideAngularDamping=true
    bUseGravityOverride=true
    gravityScale=0.35
    simulationSpace=RootRelative
    boundBone {boneName=frontjacket_A_r_02}
    componentLinearAccScale=<1 1 1>
    componentLinearVelScale=<1 1 1>
    gravityOverride {z=-980}
    physicsBodyDefinitions [{
      boundBone {boneName=frontjacket_A_r_02}
      boxExtents=<10 30 10>
      constraintSetup {angularLimitsMax=<0 0 5>, angularLimitsMin=<0 -15 -5>, angularTarget=<1 0 0>}
      localJointOffset=<2 0 0>
     }]
   }
   n2 = "Transform (Modify) Bone - Bone: base_frontjacket_A_l"(ComponentPose=Local To Component(LocalPose=Use cached pose 'InputCH'()), Alpha=1, Translation="0,0.5,0", Rotation="0,0,-3", Scale="1,1,1") {
    rotationMode=BMM_Additive
    rotationSpace=BCS_BoneSpace
    translationMode=BMM_Additive
    translationSpace=BCS_BoneSpace
    boneToModify {boneName=base_frontjacket_A_l}
    rotation {roll=-3}
    translation {y=0.5}
   }
   n3 = Anim Dynamics - Bone: frontjacket_A_l_01(ComponentPose=@n2, Alpha=0.8, ExternalForce="0,0,0") {
    angularDampingOverride=1
    angularSpringConstant=20000
    bAngularSpring=true
    bOverrideAngularDamping=true
    bUseGravityOverride=true
    gravityScale=0.35
    simulationSpace=BoneRelative
    boundBone {boneName=frontjacket_A_l_01}
    componentAppliedLinearAccClamp=<3000 3000 3000>
    componentLinearAccScale=<1 1 1>
    componentLinearVelScale=<1 1 1>
    gravityOverride {z=-980}
    physicsBodyDefinitions [{
      boundBone {boneName=frontjacket_A_l_01}
      boxExtents=<10 30 10>
      constraintSetup {angularLimitsMax=<0 0 8>, angularLimitsMin=<0 -15 -8>, angularTarget=<1 0 0>}
      localJointOffset=<6 0 0>
     }]
    relativeSpaceBone {boneName=pelvis}
    simSpaceSettings {maxAngularAcceleration=2000, maxAngularVelocity=2000, simSpaceAngularAlpha=0.6}
   }
   n4 = Anim Dynamics - Bone: frontjacket_A_l_02(ComponentPose=@n3, Alpha=0.8, ExternalForce="0,0,0") {
    angularDampingOverride=1
    angularSpringConstant=20000
    bAngularSpring=true
    bOverrideAngularDamping=true
    bUseGravityOverride=true
    gravityScale=0.35
    simulationSpace=BoneRelative
    boundBone {boneName=frontjacket_A_l_02}
    componentAppliedLinearAccClamp=<3000 3000 3000>
    componentLinearAccScale=<1 1 1>
    componentLinearVelScale=<1 1 1>
    gravityOverride {z=-980}
    physicsBodyDefinitions [{
      boundBone {boneName=frontjacket_A_l_02}
      boxExtents=<10 30 10>
      constraintSetup {angularLimitsMax=<0 0 8>, angularLimitsMin=<0 -15 -8>, angularTarget=<1 0 0>}
      localJointOffset=<6 0 0>
     }]
    relativeSpaceBone {boneName=pelvis}
    simSpaceSettings {maxAngularAcceleration=2000, maxAngularVelocity=2000, simSpaceAngularAlpha=0.6}
   }
   n5 = Anim Dynamics - Bone: frontjacket_A_r_01(ComponentPose=@n4, Alpha=0.8, ExternalForce="0,0,0") {
    angularDampingOverride=1
    angularSpringConstant=20000
    bAngularSpring=true
    bOverrideAngularDamping=true
    bUseGravityOverride=true
    gravityScale=0.35
    simulationSpace=RootRelative
    boundBone {boneName=frontjacket_A_r_01}
    componentAppliedLinearAccClamp=<3000 3000 3000>
    componentLinearAccScale=<1 1 1>
    componentLinearVelScale=<1 1 1>
    gravityOverride {z=-980}
    physicsBodyDefinitions [{
      boundBone {boneName=frontjacket_A_r_01}
      boxExtents=<10 30 10>
      constraintSetup {angularLimitsMax=<0 0 5>, angularLimitsMin=<0 -15 -5>, angularTarget=<1 0 0>}
      localJointOffset=<6 0 0>
     }]
    relativeSpaceBone {boneName=pelvis}
    simSpaceSettings {maxAngularAcceleration=2000, maxAngularVelocity=2000, simSpaceAngularAlpha=0.45}
   }
   n6 = Anim Dynamics - Bone: frontjacket_A_r_02(ComponentPose=@n5, Alpha=0.8, ExternalForce="0,0,0") {
    angularDampingOverride=1
    angularSpringConstant=20000
    bAngularSpring=true
    bOverrideAngularDamping=true
    bUseGravityOverride=true
    gravityScale=0.35
    simulationSpace=RootRelative
    boundBone {boneName=frontjacket_A_r_02}
    componentAppliedLinearAccClamp=<3000 3000 3000>
    componentLinearAccScale=<1 1 1>
    componentLinearVelScale=<1 1 1>
    gravityOverride {z=-980}
    physicsBodyDefinitions [{
      boundBone {boneName=frontjacket_A_r_02}
      boxExtents=<10 30 10>
      constraintSetup {angularLimitsMax=<0 0 5>, angularLimitsMin=<0 -15 -5>, angularTarget=<1 0 0>}
      localJointOffset=<6 0 0>
     }]
    relativeSpaceBone {boneName=pelvis}
    simSpaceSettings {maxAngularAcceleration=2000, maxAngularVelocity=2000, simSpaceAngularAlpha=0.45}
   }
   n7 = Component To Local(ComponentPose=@n6)
   n8 = Anim Dynamics - Bone: frontjacket_A_l_01(ComponentPose=Local To Component(LocalPose=Use cached pose 'InputCH'()), Alpha=0.6, ExternalForce="0,0,0") {
    angularDampingOverride=1
    angularSpringConstant=9000
    bAngularSpring=true
    bOverrideAngularDamping=true
    bUseGravityOverride=true
    gravityScale=0.35
    simulationSpace=RootRelative
    boundBone {boneName=frontjacket_A_l_01}
    componentLinearAccScale=<1 1 1>
    componentLinearVelScale=<1 1 1>
    gravityOverride {z=-980}
    physicsBodyDefinitions [{
      boundBone {boneName=frontjacket_A_l_01}
      boxExtents=<10 30 10>
      constraintSetup {angularLimitsMax=<0 0 5>, angularLimitsMin=<0 -15 -5>, angularTarget=<1 0 0>}
      localJointOffset=<2 0 0>
     }]
   }
   n9 = Anim Dynamics - Bone: frontjacket_A_l_02(ComponentPose=@n8, Alpha=0.4, ExternalForce="0,0,0") {
    angularDampingOverride=1
    angularSpringConstant=9000
    bAngularSpring=true
    bOverrideAngularDamping=true
    bUseGravityOverride=true
    gravityScale=0.35
    simulationSpace=RootRelative
    boundBone {boneName=frontjacket_A_l_02}
    componentLinearAccScale=<1 1 1>
    componentLinearVelScale=<1 1 1>
    gravityOverride {z=-980}
    physicsBodyDefinitions [{
      boundBone {boneName=frontjacket_A_l_02}
      boxExtents=<10 30 10>
      constraintSetup {angularLimitsMax=<0 0 5>, angularLimitsMin=<0 -15 -5>, angularTarget=<1 0 0>}
      localJointOffset=<2 0 0>
     }]
   }
   n10 = Component To Local(ComponentPose=@n9)
   n11 = "Constraint - base_mastr_watch (PPPPPPPP)"!(
    ComponentPose=Local To Component(LocalPose=Blend Poses by bool(BlendPose_0=@n7, BlendPose_1=@n10))
    Alpha=1
    ConstraintWeights_0=0.4306
    ConstraintWeights_1:float=0.2753
    ConstraintWeights_2:float=0.0906
    ConstraintWeights_3:float=0.0896
    ConstraintWeights_4:float=0.0611
    ConstraintWeights_5:float=0.0421
    ConstraintWeights_6:float=0.0098
    ConstraintWeights_7:float=0.0009
   ) -> (Pose) {
    boneToModify {boneName=base_mastr_watch}
    constraintSetup [
     {targetBone {boneName=lowerarm_twist_01_r}}
     {targetBone {boneName=deform_wrist_dn_r}}
     {targetBone {boneName=hand_r}}
     {targetBone {boneName=thumb_01_r}}
     {targetBone {boneName=pinky_metacarpal_r}}
     {targetBone {boneName=lowerarm_r}}
     {targetBone {boneName=ring_metacarpal_r}}
     {targetBone {boneName=middle_metacarpal_r}}
    ]
    constraintWeights [1,1,1,1,1,1,1,1]
   }
   n12 = "Constraint - base_Buttons_01 (PPPPP)"!(
    ComponentPose=@n11
    Alpha=1
    ConstraintWeights_0=0.317
    ConstraintWeights_1:float=0.2229
    ConstraintWeights_2:float=0.2137
    ConstraintWeights_3:float=0.2125
    ConstraintWeights_4:float=0.0338
   ) -> (Pose) {
    boneToModify {boneName=base_Buttons_01}
    constraintSetup [
     {targetBone {boneName=spine_02}}
     {targetBone {boneName=spine_03}}
     {targetBone {boneName=pelvis}}
     {targetBone {boneName=spine_01}}
     {targetBone {boneName=spine_04}}
    ]
    constraintWeights [1,1,1,1,1]
   }
   n13 = "Constraint - base_Buttons_02 (PPPPPPPP)"!(
    ComponentPose=@n12
    Alpha=1
    ConstraintWeights_0=0.785
    ConstraintWeights_1:float=0.0735
    ConstraintWeights_2:float=0.0631
    ConstraintWeights_3:float=0.0433
    ConstraintWeights_4:float=0.016
    ConstraintWeights_5:float=0.0112
    ConstraintWeights_6:float=0.0048
    ConstraintWeights_7:float=0.0032
   ) -> (Pose) {
    boneToModify {boneName=base_Buttons_02}
    constraintSetup [
     {targetBone {boneName=pelvis}}
     {targetBone {boneName=spine_02}}
     {targetBone {boneName=spine_01}}
     {targetBone {boneName=spine_03}}
     {targetBone {boneName=deform_groin_r}}
     {targetBone {boneName=thigh_twist_01_r}}
     {targetBone {boneName=thigh_twist_01_l}}
     {targetBone {boneName=thigh_r_jacket_deform}}
    ]
    constraintWeights [1,1,1,1,1,1,1,1]
   }
   n14 = RigidBody!(ComponentPose=@n13, Alpha, OverrideWorldGravity:Vector="0,0,0", ExternalForce) -> (Pose) {
    bOverrideWorldGravity=true
    overridePhysicsAsset=&/M_MED_CiderPetalFlax_Physics
    simulationSpace=BaseBoneSpace
    alphaBoolBlend {blendInTime=0.2, blendOutTime=0.2}
    baseBoneRef {boneName=pelvis}
    componentAppliedLinearAccClamp=<0 0 0>
    simSpaceSettings {
     dampingAlpha=0
     maxAngularAcceleration=4
     maxAngularVelocity=5
     maxLinearAcceleration=150
     maxLinearVelocity=300
     worldAlpha=1
     externalLinearDragV=<1 1 1>
    }
   }
   Output Pose(Result=Component To Local(ComponentPose=@n14)) {layerGroup=None}
   n15 = Control Rig!(Source=Copy Pose From Mesh() {bUseAttachedParent=true}, Alpha, ISCrounching:bool=false) -> (Pose) {controlRigAssetReference {blueprintRigClass=&/M_MED_CiderPetalFlax_CtrlRig.M_MED_CiderPetalFlax_CtrlRig_C}}
   Save cached pose 'InputCH'(Pose=@n15) {cacheName=InputCH}
  ]}
 EventGraph {nodes [
   Event BlueprintUpdateAnimation()
   > n1 = "Gravity Override State Function (Parameter Asset) (Nativized)"(
    CharacterPartAnimInstance=Self-Reference()
    GameTimeInSeconds=GetGameTimeInSeconds()
    Parameters=/BRCosmetics/Characters/Player/Female/Medium/Bodies/F_MED_AlderOptic/Meshes/F_MED_AlderOptic_FaceAcc_Parameters
   )
   > Set Gravity Override(Gravity Override=@n1)
   > Set IsCrounchingActive(IsCrounchingActive:bool=Get bIsCrouching()) -> (Output_Get:bool)
  ]}
]
variables [
 Gravity Override {type=Vector}
 IsCrounchingActive {type=boolean}
]
