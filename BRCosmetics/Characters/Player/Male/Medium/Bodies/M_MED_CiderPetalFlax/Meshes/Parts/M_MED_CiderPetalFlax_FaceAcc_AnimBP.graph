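# Animation Blueprint for the CiderPetalFlax face accessory mesh, adding secondary hair motion to the player cosmetic part.
# It copies the parent mesh pose, constrains base_hair_a to head, then chains spring-driven Anim Dynamics on three hair bones.
# The event graph computes Gravity Override from CiderPetalClover's parameter asset, though the dynamics nodes show fixed gravity values rather than that variable.
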AnimBlueprint /BRCosmetics/Characters/Player/Male/Medium/Bodies/M_MED_CiderPetalFlax/Meshes/Parts/M_MED_CiderPetalFlax_FaceAcc_AnimBP
target_skeleton=&/M_MED_CiderPetalFlax_FaceAcc_Skeleton
graphs [
 AnimGraph {nodes [
   Output Pose(Result=Use cached pose 'DynCH'()) {layerGroup=None}
   Save cached pose 'InputCH'(Pose=Copy Pose From Mesh() {bUseAttachedParent=true}) {cacheName=InputCH}
   n1 = "Constraint - base_hair_a (P)"!(ComponentPose=Local To Component(LocalPose=Use cached pose 'InputCH'()), Alpha=1, ConstraintWeights_0=1, ConstraintWeights_1:float=0) -> (Pose) {boneToModify {boneName=base_hair_a}, constraintSetup [{targetBone {boneName=head}}], constraintWeights [1,0]}
   n2 = Anim Dynamics - Bone: dyn_hair_a!(ComponentPose=@n1, Alpha=0.8, GravityOverride:Vector="0,0,-980", ExternalForce="0,0,0") -> (Pose) {
    angularDampingOverride=1
    angularSpringConstant=8000
    bAngularSpring=true
    bOverrideAngularDamping=true
    bUseGravityOverride=true
    gravityScale=0.35
    simulationSpace=RootRelative
    alphaBoolBlend {blendOption=Linear}
    boundBone {boneName=dyn_hair_a}
    componentAppliedLinearAccClamp=<3000 3000 3000>
    componentLinearAccScale=<1 1 1>
    componentLinearVelScale=<1 1 1>
    gravityOverride {z=-980}
    physicsBodyDefinitions [{
      boundBone {boneName=dyn_hair_a}
      boxExtents=<15 25 25>
      constraintSetup {angularLimitsMax=<0 10 -15>, angularLimitsMin=<0 -10 -15>, angularTarget=<1 0 0>}
      localJointOffset=<4 0 0>
     }]
   }
   n3 = Anim Dynamics - Bone: dyn_hairStrandMain_1!(ComponentPose=@n2, Alpha=0.4, GravityOverride:Vector="0,0,-980", ExternalForce="0,0,0") -> (Pose) {
    angularDampingOverride=1
    angularSpringConstant=15000
    bAngularSpring=true
    bOverrideAngularDamping=true
    bUseGravityOverride=true
    gravityScale=0.35
    simulationSpace=RootRelative
    alphaBoolBlend {blendOption=Linear}
    boundBone {boneName=dyn_hairStrandMain_1}
    componentAppliedLinearAccClamp=<3000 3000 3000>
    componentLinearAccScale=<0.4 0.4 0.4>
    componentLinearVelScale=<0.4 0.4 0.4>
    gravityOverride {z=-980}
    physicsBodyDefinitions [{
      boundBone {boneName=dyn_hairStrandMain_1}
      boxExtents=<20 30 30>
      constraintSetup {angularLimitsMax=<0 5 10>, angularLimitsMin=<0 -5 -10>, angularTarget=<1 0 0>}
      localJointOffset=<0.3 0 0>
     }]
   }
   n4 = Anim Dynamics - Bone: dyn_hairStrandMain_2!(ComponentPose=@n3, Alpha=0.5, GravityOverride:Vector="0,0,-980", ExternalForce="0,0,0") -> (Pose) {
    angularDampingOverride=1
    angularSpringConstant=15000
    bAngularSpring=true
    bOverrideAngularDamping=true
    bUseGravityOverride=true
    gravityScale=0.35
    simulationSpace=RootRelative
    alphaBoolBlend {blendOption=Linear}
    boundBone {boneName=dyn_hairStrandMain_2}
    componentAppliedLinearAccClamp=<3000 3000 3000>
    componentLinearAccScale=<0.4 0.4 0.4>
    componentLinearVelScale=<0.4 0.4 0.4>
    gravityOverride {z=-980}
    physicsBodyDefinitions [{
      boundBone {boneName=dyn_hairStrandMain_2}
      boxExtents=<20 30 30>
      constraintSetup {angularLimitsMax=<0 5 10>, angularLimitsMin=<0 -5 -10>, angularTarget=<1 0 0>}
      localJointOffset=<0.3 0 0>
     }]
   }
   Save cached pose 'DynCH'(Pose=Component To Local(ComponentPose=@n4)) {cacheName=DynCH}
  ]}
 EventGraph {nodes [
   Event BlueprintUpdateAnimation()
   > n1 = "Gravity Override State Function (Parameter Asset) (Nativized)"(
    CharacterPartAnimInstance=Self-Reference()
    GameTimeInSeconds=GetGameTimeInSeconds()
    Parameters=/BRCosmetics/Characters/Player/Male/Medium/Bodies/M_MED_CiderPetalClover/Meshes/Parts/M_MED_CiderPetalClover_FaceAcc_Parameters
   )
   > Set Gravity Override(Gravity Override=@n1)
  ]}
]
variables [Gravity Override {type=Vector}]
